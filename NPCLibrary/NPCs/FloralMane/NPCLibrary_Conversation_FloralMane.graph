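# Conversation database for the FloralMane NPC, a movie director who surfs, covering his greetings, small talk and dialogue choices.
# The entry greeting picks contextual lines by first meeting, match conversation count or worn outfit, using priority curves and weights.
# It branches to an "Ask About..." menu with multi-step Surfing and Peelians chats that return to start.
# Link nodes pull in shared quest, basic-service and Season 39 PerfectGift conversations.
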ConversationDatabase /NPCLibrary/NPCs/FloralMane/NPCLibrary_Conversation_FloralMane
entry_points [{tag=Conversation.Entry.FloralMane, destinations [#1]}]
nodes [
 ConversationEntryPointNode_7 {
  id=#1
  class=ConversationEntryPointNode
  outputs [#2]
  properties {entryTag {tagName=Conversation.Entry.FloralMane}}
 }
 Speech_C_4 {
  id=#2
  class=Speech_C
  outputs [#3, #4, #5, #6]
  properties {
   nodeName=""
   generalConfig {
    defaultMessage=Hey. You a fan of my movies?
    contextualMessages [
     {
      message=Hey. You a fan of my movies?
      contextRequirements [{requirement {class=FortControllerRequirement_IsFirstEverConversationWithNPC}}]
      priority {curve {rowName=Tandem.Speech.Priority.FirstEverInteraction}}
     }
     {
      message=Any movie longer than ninety minutes needs an editor or a chainsaw.
      contextRequirements [{requirement {
         class=FortControllerRequirement_CheckNPCMatchConversationCount
         properties {nPCUniqueIdTag {tagName=AISpawnerData.Type.Specialist.Medic.FloralMane}}
        }}]
      priority {curve {curveTable=null, rowName=None}}
     }
     {
      message=You ever think about becoming an actor?
      contextRequirements [{requirement {
         class=FortControllerRequirement_CheckNPCMatchConversationCount
         properties {nPCUniqueIdTag {tagName=AISpawnerData.Type.Specialist.Medic.FloralMane}}
        }}]
      priority {curve {curveTable=null, rowName=None}}
     }
     {
      message=You've seen a few of my movies. Trust me.
      contextRequirements [{requirement {
         class=FortControllerRequirement_CheckNPCMatchConversationCount
         properties {nPCUniqueIdTag {tagName=AISpawnerData.Type.Specialist.Medic.FloralMane}}
        }}]
      priority {curve {curveTable=null, rowName=None}}
     }
     {
      message="There's my star! No finer action man in the business!"
      contextRequirements [{requirement {
         class=FortControllerRequirement_HasCID
         properties {allowedCIds [/BRCosmetics/Athena/Items/Cosmetics/Characters/Character_DerangedMile_Side]}
        }}]
      priority {curve {rowName=Tandem.Speech.Priority.OutfitReaction}}
     }
     {
      message=There's an action heroine inside you. You just gotta let her roar.
      contextRequirements [{requirement {class=FortControllerRequirement_HasCID, properties {allowedCIds [/BRCosmetics/Athena/Items/Cosmetics/Characters/Character_RowLiaison_Cafe]}}}]
      priority {curve {rowName=Tandem.Speech.Priority.OutfitReaction}}
     }
     {
      message="Wu! Still can't believe The Golden Dragon himself is on my set."
      contextRequirements [{requirement {class=FortControllerRequirement_HasCID, properties {allowedCIds [/BRCosmetics/Athena/Items/Cosmetics/Characters/Character_ZenCrash]}}}]
      priority {curve {rowName=Tandem.Speech.Priority.OutfitReaction}}
     }
     {
      message="Be careful, Peels. Lotta hungry people in this industry that'll make you a meal."
      contextRequirements [{requirement {
         class=FortControllerRequirement_HasCID
         properties {allowedCIds [
           &Character_BananaPhilosopher
           &Character_BananaAdventure
           /BRCosmetics/Athena/Items/Cosmetics/Characters/Character_BananaCake
           &CID_349_Athena_Commando_M_Banana
           &CID_701_Athena_Commando_M_BananaAgent
           &CID_814_Athena_Commando_M_BananaSummer
           /BRCosmetics/Athena/Items/Cosmetics/Characters/CID_A_045_Athena_Commando_M_BananaLeader
           &CID_A_323_Athena_Commando_M_BananaWinter
           /BRCosmetics/Athena/Items/Cosmetics/Characters/CID_TBD_Athena_Commando_M_Banana_CINE
          ]}
        }}]
      priority {curve {rowName=Tandem.Speech.Priority.OutfitReaction}}
     }
     {
      message="\"Films\" are for snobs. I'm a movie guy."
      contextRequirements [{requirement {
         class=FortControllerRequirement_CheckNPCMatchConversationCount
         properties {nPCUniqueIdTag {tagName=AISpawnerData.Type.Specialist.Medic.FloralMane}}
        }}]
      priority {curve {curveTable=null, rowName=None}}
     }
     {
      message=Waves have been choppy lately. Something's comin'.
      contextRequirements [{requirement {
         class=FortControllerRequirement_CheckNPCMatchConversationCount
         properties {nPCUniqueIdTag {tagName=AISpawnerData.Type.Specialist.Medic.FloralMane}}
        }}]
      priority {curve {curveTable=null, rowName=None}}
      weight {value=2}
     }
     {
      message="Movies teach us to be heroes. Remember that, when things get dark."
      contextRequirements [{requirement {
         class=FortControllerRequirement_CheckNPCMatchConversationCount
         properties {nPCUniqueIdTag {tagName=AISpawnerData.Type.Specialist.Medic.FloralMane}}
        }}]
      priority {curve {curveTable=null, rowName=None}}
      weight {value=2}
     }
     {
      message=I really think you're going to love Peelians.
      contextRequirements [{requirement {
         class=FortControllerRequirement_CheckNPCMatchConversationCount
         properties {nPCUniqueIdTag {tagName=AISpawnerData.Type.Specialist.Medic.FloralMane}}
        }}]
      priority {curve {curveTable=null, rowName=None}}
     }
    ]
   }
  }
 }
 ConversationLinkNode_2 {
  id=#6
  class=ConversationLinkNode
  properties {remoteEntryTag {tagName=Conversation.Entry.Season39.WF.PerfectGift}}
 }
 Speech_C_2 {
  id=#5
  class=Speech_C
  outputs [#7, #8, #9]
  properties {nodeName="", softTaskIcon=&T_UI_NPC_question, generalConfig {defaultMessage=What do you need?}}
  sub_nodes [ConversationChoiceNode_0 {class=ConversationChoiceNode, properties {defaultChoiceDisplayText=Ask About...}}]
 }
 Speech_C_1 {
  id=#9
  class=Speech_C
  outputs [#10]
  properties {
   nodeName=""
   softTaskIcon=&T_UI_NPC_question
   generalConfig {defaultMessage="Trust me, it's not gonna win a Forté Award, and we're not trying to."}
  }
  sub_nodes [ConversationChoiceNode_41 {class=ConversationChoiceNode, properties {defaultChoiceDisplayText=Peelians}}]
 }
 Speech_C_6 {
  id=#10
  class=Speech_C
  outputs [#11]
  properties {
   nodeName=""
   softTaskIcon=&T_UI_NPC_question
   generalConfig {defaultMessage=All I want is for people watching it to have a total blast. I don't need critical acclaim or whatever.}
  }
  sub_nodes [ConversationChoiceNode_43 {class=ConversationChoiceNode, properties {defaultChoiceDisplayText=Continue...}}]
 }
 Speech_C_8 {
  id=#11
  class=Speech_C
  outputs [#12]
  properties {
   nodeName=""
   softTaskIcon=&T_UI_NPC_question
   generalConfig {defaultMessage="Miles Cross rams a reboot van full of ethylene gas into a hive of banana-alien hybrids. That's the movie!"}
  }
  sub_nodes [ConversationChoiceNode_44 {class=ConversationChoiceNode, properties {defaultChoiceDisplayText=Continue...}}]
 }
 Speech_C_9 {
  id=#12
  class=Speech_C
  outputs [#13]
  properties {
   nodeName=""
   softTaskIcon=&T_UI_NPC_question
   generalConfig {defaultMessage="It's fun, it's ridiculous, and you'll wanna see it in theatres five times with all your friends."}
  }
  sub_nodes [ConversationChoiceNode_45 {class=ConversationChoiceNode, properties {defaultChoiceDisplayText=Continue...}}]
 }
 ReturnToStart_C_0 {
  id=#13
  class=ReturnToStart_C
  properties {softTaskIcon=&T_UI_NPC_Back}
  sub_nodes [ConversationChoiceNode_42 {class=ConversationChoiceNode, properties {defaultChoiceDisplayText=BACK}}]
 }
 Back_C_2 {
  id=#8
  class=Back_C
  sub_nodes [ConversationChoiceNode_40 {class=ConversationChoiceNode, properties {defaultChoiceDisplayText=BACK}}]
 }
 Speech_C_11 {
  id=#7
  class=Speech_C
  outputs [#14]
  properties {
   nodeName=""
   softTaskIcon=&T_UI_NPC_question
   generalConfig {defaultMessage="People always say: surfing is chill, directing isn't. How do you reconcile that?"}
  }
  sub_nodes [ConversationChoiceNode_1 {class=ConversationChoiceNode, properties {defaultChoiceDisplayText=Surfing}}]
 }
 Speech_C_3 {
  id=#14
  class=Speech_C
  outputs [#15]
  properties {
   nodeName=""
   softTaskIcon=&T_UI_NPC_question
   generalConfig {defaultMessage="Easy. On set, I'm focused. Every single second."}
  }
  sub_nodes [ConversationChoiceNode_3 {class=ConversationChoiceNode, properties {defaultChoiceDisplayText=Continue...}}]
 }
 Speech_C_5 {
  id=#15
  class=Speech_C
  outputs [#16]
  properties {
   nodeName=""
   softTaskIcon=&T_UI_NPC_question
   generalConfig {defaultMessage="Which is great, but exhausting. Surfing's my outlet."}
  }
  sub_nodes [ConversationChoiceNode_4 {class=ConversationChoiceNode, properties {defaultChoiceDisplayText=Continue...}}]
 }
 Speech_C_7 {
  id=#16
  class=Speech_C
  outputs [#17]
  properties {
   nodeName=""
   softTaskIcon=&T_UI_NPC_question
   generalConfig {defaultMessage="Out on the waves, I can roar as loud as I like. The ocean's always louder."}
  }
  sub_nodes [ConversationChoiceNode_5 {class=ConversationChoiceNode, properties {defaultChoiceDisplayText=Continue...}}]
 }
 ReturnToStart_C_23 {
  id=#17
  class=ReturnToStart_C
  properties {softTaskIcon=&T_UI_NPC_Back}
  sub_nodes [ConversationChoiceNode_2 {class=ConversationChoiceNode, properties {defaultChoiceDisplayText=BACK}}]
 }
 ConversationLinkNode_0 {id=#4, class=ConversationLinkNode, properties {remoteEntryTag {tagName=Conversation.Entry.BasicServices}}}
 ConversationLinkNode_1 {id=#3, class=ConversationLinkNode, properties {remoteEntryTag {tagName=Conversation.Entry.NPCLibrary.Quests}}}
]
